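# Editor utility widget in the Plugin Connection Finder tools that displays one asset as a thumbnail with its package name below.
# SetAssetPackageName stores the path, resolves it to asset data through an Asset Registry lookup macro, then updates the thumbnail and text only if a match is found.
# A find-icon button syncs the Content Browser to the stored asset path.
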EditorUtilityWidgetBlueprint /Game/NeverCook/EditorUtilities/ThinClientTools/PluginConnectionFinder/EUW_PCF_ConnectionDisplayThumb
graphs [
 EventGraph {nodes [
   "On Clicked (EditorUtilityButton_15)"() {is=title K2Node_ComponentBoundEvent}
   > SyncBrowserToObjects(AssetPaths=Make Array("[0]":String=Get CurrentAsset() -> (CurrentAsset:String)) -> (Array:Array of Strings))
  ]}
 SetAssetPackageName {nodes [
   n1 = SetAssetPackageName!() -> (then, AssetPackageName:String) {is=title K2Node_FunctionEntry}
   > n2 = Set CurrentAsset(CurrentAsset:String=@n1) -> (Output_Get:String)
   > n3 = Try Get Asset from Package Path!(execute:Exec, InString:String=@n2) -> (true:Exec, false:Exec, Output:Asset Data Structure) {is=title K2Node_MacroInstance}
   > SetAsset!(
    execute
    self:Asset Thumbnail Widget Object Reference=Get thumbnail() -> (thumbnail:Asset Thumbnail Widget Object Reference)
    AssetData:"Asset Data Structure (by ref)"=@n3
   ) -> (then)
   > "SetText (Text)"(self=Get txtAssetPackageName() -> (txtAssetPackageName:Text Object Reference), InText="To Text (String)"(InString=Get CurrentAsset()))
  ]}
 TryGetAssetFromPackagePath {nodes [
   n1 = Inputs!() -> (execute, InString:String)
   > n2 = GetAssetsByPackageName!(
    execute
    self:Asset Registry Interface=GetAssetRegistry()
    PackageName:Name=String To Name(InString=@n1)
    bIncludeOnlyOnDiskAssets:bool=false
    bSkipARFilteredAssets:bool=true
   ) -> (then, OutAssetData:Array of Asset Data Structures, ReturnValue:bool)
   > n3 = Branch(Condition=Is Valid Index!(TargetArray:Array of Asset Data Structures=@n2, IndexToTest:int) -> (ReturnValue:bool))
   > Outputs!(
    true:Exec
    false:Exec=@n3.else
    Output:Asset Data Structure="Get (a copy)"(Array:Array of Asset Data Structures=@n2) -> (Output:Asset Data Structure)
   )
  ]}
]
variables [
 CurrentAsset {type=string, default=""}
 CurrentDataObject {type=O_PCF_ConnectionDisplayData_C, default=null}
]
widgets {
 info {rootWidgetClass=&VerticalBox, widgetCount=6}
 widgets [
  VerticalBox {widgetClassPath=&VerticalBox, widget=VerticalBox}
  HorizontalBox_0 {
   widgetClassPath=&HorizontalBox
   parent=VerticalBox
   widget=HorizontalBox_0
   slot {class=VerticalBoxSlot, properties {horizontalAlignment=HAlign_Center}}
  }
  thumbnail {
   widgetClassPath=/Script/UMGEditor.AssetThumbnailWidget
   bIsVariable=true
   parent=HorizontalBox_0
   widget=thumbnail
   properties {thumbnailSettings {
     bAllowAssetSpecificThumbnailOverlay=false
     bAllowHintText=true
     bAllowRealTimeOnHovered=false
     bForceGenericThumbnail=false
     bOverrideAssetTypeColor=false
     colorStripOrientation=HorizontalBottomEdge
     genericThumbnailSize=64
     highlightedTextDelegate="(null).None"
     thumbnailLabel=AssetName
     assetTypeColorOverride=<rgba 0 0 0 0>
     hintColorAndOpacity=<rgba 0 0 0 0>
     padding=<ltrb 0 0 0 0>
    }}
   slot {class=HorizontalBoxSlot}
  }
  EditorUtilityButton_15 {
   widgetClassPath=&EditorUtilityButton
   parent=HorizontalBox_0
   widget=EditorUtilityButton_15
   properties {widgetStyle {
     clickedSlateSound {resourceObject=null}
     disabled {
      bIsDynamicallyLoaded=false
      drawAs=RoundedBox
      imageType=NoImage
      mirroring=NoMirror
      resourceName=None
      resourceObject=null
      tiling=NoTile
      imageSize=<0 0>
      margin=<ltrb 0 0 0 0>
      outlineSettings {
       bUseBrushTransparency=false
       roundingType=FixedRadius
       width=1
       color {colorUseRule=UseColor_ColorTable, specifiedColor=<rgba 0 0 0 0>}
       cornerRadii=<4 4 4 4>
      }
      tintColor {colorUseRule=UseColor_Specified, specifiedColor=<rgba 0.0395 0.0395 0.0395 1>}
      uVRegion {bIsValid=false, max=<0 0>, min=<0 0>}
     }
     disabledForeground {colorUseRule=UseColor_ColorTable, specifiedColor=<rgba 0 0 0 0>}
     hovered {
      bIsDynamicallyLoaded=false
      drawAs=RoundedBox
      imageType=NoImage
      mirroring=NoMirror
      resourceName=None
      resourceObject=null
      tiling=NoTile
      imageSize=<32 32>
      margin=<ltrb 0 0 0 0>
      outlineSettings {
       bUseBrushTransparency=false
       roundingType=FixedRadius
       width=1
       color {colorUseRule=UseColor_ColorTable, specifiedColor=<rgba 0 0 0 0>}
       cornerRadii=<4 4 4 4>
      }
      tintColor {colorUseRule=UseColor_Specified, specifiedColor=<rgba 0.0953 0.0953 0.0953 1>}
      uVRegion {bIsValid=false, max=<0 0>, min=<0 0>}
     }
     hoveredForeground {colorUseRule=UseColor_ColorTable, specifiedColor=<rgba 0 0 0 0>}
     hoveredSlateSound {resourceObject=null}
     normal {
      bIsDynamicallyLoaded=false
      drawAs=RoundedBox
      imageType=NoImage
      mirroring=NoMirror
      resourceName=None
      resourceObject=null
      tiling=NoTile
      imageSize=<32 32>
      margin=<ltrb 0 0 0 0>
      outlineSettings {
       bUseBrushTransparency=false
       roundingType=FixedRadius
       width=1
       color {colorUseRule=UseColor_ColorTable, specifiedColor=<rgba 0 0 0 0>}
       cornerRadii=<4 4 4 4>
      }
      tintColor {colorUseRule=UseColor_Specified, specifiedColor=<rgba 0.0395 0.0395 0.0395 1>}
      uVRegion {bIsValid=false, max=<0 0>, min=<0 0>}
     }
     normalForeground {colorUseRule=UseColor_ColorTable, specifiedColor=<rgba 0 0 0 0>}
     normalPadding=<ltrb 0 0 0 0>
     pressed {
      bIsDynamicallyLoaded=false
      drawAs=RoundedBox
      imageType=NoImage
      mirroring=NoMirror
      resourceName=None
      resourceObject=null
      tiling=NoTile
      imageSize=<32 32>
      margin=<ltrb 0 0 0 0>
      outlineSettings {
       bUseBrushTransparency=false
       roundingType=FixedRadius
       width=1
       color {colorUseRule=UseColor_ColorTable, specifiedColor=<rgba 0 0 0 0>}
       cornerRadii=<4 4 4 4>
      }
      tintColor {colorUseRule=UseColor_Specified, specifiedColor=<rgba 0.0284 0.0284 0.0284 1>}
      uVRegion {bIsValid=false, max=<0 0>, min=<0 0>}
     }
     pressedForeground {colorUseRule=UseColor_ColorTable, specifiedColor=<rgba 0 0 0 0>}
     pressedPadding=<ltrb 0 0 0 0>
     pressedSlateSound {resourceObject=null}
    }}
   slot {class=HorizontalBoxSlot, properties {horizontalAlignment=HAlign_Left, verticalAlignment=VAlign_Center, padding=<ltrb 15 0 0 0>}}
  }
  Image_40 {
   widgetClassPath=&Image
   bIsVariable=true
   parent=EditorUtilityButton_15
   widget=Image_40
   properties {brush {
     bIsDynamicallyLoaded=false
     drawAs=Image
     imageType=FullColor
     mirroring=NoMirror
     resourceName=None
     resourceObject=/Game/Athena/UI/Challenges/Art/ChallengeTypeIcons/UI_ChallengeIcon_Find
     tiling=NoTile
     imageSize=<32 32>
     margin=<ltrb 0 0 0 0>
     outlineSettings {
      bUseBrushTransparency=false
      roundingType=HalfHeightRadius
      width=0
      color {colorUseRule=UseColor_Specified, specifiedColor=<rgba 0 0 0 0>}
      cornerRadii=<0 0 0 0>
     }
     tintColor {colorUseRule=UseColor_Specified, specifiedColor=<rgba 1 1 1 1>}
     uVRegion {bIsValid=false, max=<0 0>, min=<0 0>}
    }}
   slot {class=ButtonSlot, properties {padding=<ltrb 3 3 3 3>}}
  }
  txtAssetPackageName {
   widgetClassPath=&TextBlock
   bIsVariable=true
   parent=VerticalBox
   widget=txtAssetPackageName
   properties {
    text=Text Block
    font {
     bForceMonospaced=false
     bMaterialIsStencil=false
     fontMaterial=null
     fontObject=&Roboto
     letterSpacing=0
     monospacedWidth=1
     size=9
     skewAmount=0
     typefaceFontName=Regular
     outlineSettings {
      bApplyOutlineToDropShadows=false
      bMiteredCorners=false
      bSeparateFillAlpha=false
      outlineMaterial=null
      outlineSize=0
      outlineColor=<rgba 0 0 0 1>
     }
    }
   }
   slot {class=VerticalBoxSlot, properties {horizontalAlignment=HAlign_Center}}
  }
 ]
}
