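# Editor utility list entry for the Plugin Connection Finder that shows one connection as a source thumbnail, an arrow and a target thumbnail.
# When its list item is set, it casts to O_PCF_ConnectionDisplayData, stores it and passes the source and target package names to the two thumb widgets.
# A TryGetAssetFromPackagePath helper looks up Asset Registry data by package name with a found/not-found branch, but the event graph does not use it.
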EditorUtilityWidgetBlueprint /Game/NeverCook/EditorUtilities/ThinClientTools/PluginConnectionFinder/EUW_PCF_ConnectionDisplayItem
graphs [
 EventGraph {nodes [
   n1 = Event OnListItemObjectSet()
   > n2 = Cast To O_PCF_ConnectionDisplayData!(execute, Object=@n1.ListItemObject) -> (then, CastFailed, AsO PCF Connection Display Data:O PCF Connection Display Data Object Reference) {is=title K2Node_DynamicCast}
   > Set CurrentDataObject(CurrentDataObject:O PCF Connection Display Data Object Reference=@n2) -> (Output_Get:O PCF Connection Display Data Object Reference)
   n3 = Get CurrentDataObject() -> (CurrentDataObject:O PCF Connection Display Data Object Reference)
   > SetAssetPackageName!(
    execute
    self:EUW PCF Connection Display Thumb Object Reference=Get thumbSource() -> (thumbSource:EUW PCF Connection Display Thumb Object Reference)
    AssetPackageName:String=Get source_package_name!(self:O PCF Connection Display Data Object Reference=@n3) -> (source_package_name:String)
   ) -> (then)
   > SetAssetPackageName(
    self=Get thumbTarget() -> (thumbTarget:EUW PCF Connection Display Thumb Object Reference)
    AssetPackageName=Get target_package_name!(self:O PCF Connection Display Data Object Reference=@n3) -> (target_package_name:String)
   )
   Event PreConstruct()
   Event Construct()
   Event Tick()
  ]}
 TryGetAssetFromPackagePath {nodes [
   n1 = Inputs!() -> (execute, InString:String)
   > n2 = GetAssetsByPackageName!(
    execute
    self:Asset Registry Interface=GetAssetRegistry()
    PackageName:Name=String To Name(InString=@n1)
    bIncludeOnlyOnDiskAssets:bool=false
    bSkipARFilteredAssets:bool=true
   ) -> (then, OutAssetData:Array of Asset Data Structures, ReturnValue:bool)
   > n3 = Branch(Condition=Is Valid Index!(TargetArray:Array of Asset Data Structures=@n2, IndexToTest:int) -> (ReturnValue:bool))
   > Outputs!(
    true:Exec
    false:Exec=@n3.else
    Output:Asset Data Structure="Get (a copy)"(Array:Array of Asset Data Structures=@n2) -> (Output:Asset Data Structure)
   )
  ]}
]
variables [CurrentDataObject {type=O_PCF_ConnectionDisplayData_C, default=null}]
widgets {
 info {rootWidgetClass=&StackBox, widgetCount=6}
 widgets [
  StackBox_0 {widgetClassPath=&StackBox, widget=StackBox_0, properties {orientation=Orient_Vertical}}
  Border_12 {
   widgetClassPath=&Border
   parent=StackBox_0
   widget=Border_12
   properties {brushColor=<rgba 0.0176 0.0176 0.0176 1>}
   slot {class=StackBoxSlot}
  }
  HorizontalBox_0 {widgetClassPath=&HorizontalBox, parent=Border_12, widget=HorizontalBox_0, slot {class=BorderSlot}}
  thumbSource {
   widgetClassPath=&/EUW_PCF_ConnectionDisplayThumb
   bIsVariable=true
   parent=HorizontalBox_0
   widget=thumbSource
   slot {class=HorizontalBoxSlot, properties {size {sizeRule=Fill, value=1}}}
  }
  TextBlock_10 {
   widgetClassPath=&TextBlock
   parent=HorizontalBox_0
   widget=TextBlock_10
   properties {
    text="->"
    font {
     bForceMonospaced=false
     bMaterialIsStencil=false
     fontMaterial=null
     fontObject=&Roboto
     letterSpacing=0
     monospacedWidth=1
     size=9
     skewAmount=0
     typefaceFontName=Regular
     outlineSettings {
      bApplyOutlineToDropShadows=false
      bMiteredCorners=false
      bSeparateFillAlpha=false
      outlineMaterial=null
      outlineSize=0
      outlineColor=<rgba 0 0 0 1>
     }
    }
   }
   slot {class=HorizontalBoxSlot, properties {horizontalAlignment=HAlign_Center, verticalAlignment=VAlign_Center, padding=<ltrb 20 0 20 0>}}
  }
  thumbTarget {
   widgetClassPath=&EUW_PCF_ConnectionDisplayThumb
   bIsVariable=true
   parent=HorizontalBox_0
   widget=thumbTarget
   slot {class=HorizontalBoxSlot, properties {size {sizeRule=Fill, value=1}}}
  }
 ]
}
