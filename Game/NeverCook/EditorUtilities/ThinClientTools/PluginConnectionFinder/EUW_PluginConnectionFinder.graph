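# Editor utility widget that finds asset references from a source plugin into a target plugin, both named in an embedded details view.
# GetConnections queries Asset Registry dependencies of source assets, matching target path prefixes and recursively searching material functions, with a searched set to skip repeats.
# Results form a source-to-target string map, so each source asset keeps only one target.
# PopulateDisplayList wraps each pair in a data object for the list view.
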EditorUtilityWidgetBlueprint /Game/NeverCook/EditorUtilities/ThinClientTools/PluginConnectionFinder/EUW_PluginConnectionFinder
graphs [
 EventGraph {nodes [
   "On Clicked (btnFindConnections)"!() -> (OutputDelegate, then, widget:User Widget Object Reference, event:Pointer Event Structure) {is=title K2Node_ComponentBoundEvent}
   > n1 = GetConnections!(execute, self) -> (then, SourceTargetMap:Map of Strings to Strings)
   > PopulateDisplayList!(execute, self, SourceTargetMap:Map of Strings to Strings=@n1) -> (then)
   Event Construct()
   > SetObject(self=Get detailsSourceAndTarget() -> (detailsSourceAndTarget:Details View Object Reference), NewObject=Self-Reference())
  ]}
 GetConnections {nodes [
   n1 = Get l_targetData() -> (l_targetData:Asset Data Structure)
   n2 = Break Asset Data(AssetData=@n1)
   n3 = "To String (Name)"(InName=@n2)
   n4 = PrintString(InString=Append(A="Checking MatFunc: ", B=@n3))
   n5 = GetMaterialFunctionDependenciesInTarget!(
    execute=[@n4.then, @n20.else]
    self
    InAssetData:"Asset Data Structure (by ref)"=@n1
    TargetPlugin:String=Append(A=/, B=Get TargetPluginName() -> (TargetPluginName:String))
    SearchedRefs:Set of Strings=Get l_searchedMatFunctionRefs() -> (l_searchedMatFunctionRefs:Set of Strings)
   ) -> (then, FoundRefsToTarget:Set of Strings)
   > n6 = To Array!(execute, A:Set of Strings=@n5) -> (then, Result:Array of Strings)
   > n7 = For Each Loop(Array=@n6)
   > Add!(
    execute
    TargetMap:Map of Strings to Strings=Get l_SourceTargetMap() -> (l_SourceTargetMap:Map of Strings to Strings)
    Key:"String (by ref)"=Get l_source() -> (l_source:String)
    Value:"String (by ref)"=@n7
   ) -> (then) {is=title K2Node_CallFunction}
   GetConnections() {is=title K2Node_FunctionEntry}
   n8 = String To Name(InString=Append(A=/, B=Get SourcePluginName() -> (SourcePluginName:String)))
   > n9 = GetAssets!(
    execute
    self:Asset Registry Interface=GetAssetRegistry()
    Filter:"ARFilter Structure (by ref)"=MakeARFilter!(
     PackageNames:Array of Names
     PackagePaths:Array of Names=Make Array("[0]":Name=@n8) -> (Array:Array of Names)
     SoftObjectPaths:Array of Soft Object Path Structures
     ClassPaths:Array of Top Level Asset Path Structures
     RecursiveClassPathsExclusionSet:Set of Top Level Asset Path Structures
     ClassNames:Array of Names
     RecursiveClassesExclusionSet:Set of Names
     bRecursivePaths:bool=true
     bRecursiveClasses:bool=false
     bIncludeOnlyOnDiskAssets:bool=false
    ) -> (ReturnValue:ARFilter Structure)
    bSkipARFilteredAssets:bool=true
   ) -> (then, OutAssetData:Array of Asset Data Structures, ReturnValue:bool)
   n10 = Length(TargetArray:Array of Asset Data Structures=@n9)
   > PrintString(InString=Append!(A="Checking ", B="To String (Integer)"(InInt=@n10), C:String=" assets in ", D:String=Get SourcePluginName()) -> (ReturnValue))
   > Branch(Condition=Is Valid Index!(TargetArray:Array of Asset Data Structures=@n9, IndexToTest:int) -> (ReturnValue:bool))
   > n11 = For Each Loop(Array:Array of Asset Data Structures=@n9) -> (Array Element:Asset Data Structure)
   n12 = Length!(TargetMap:Map of Strings to Strings=Get l_SourceTargetMap()) -> (ReturnValue:int=0) {is=title K2Node_CallFunction}
   PrintString(execute=@n11.Completed, InString=Append!(A="Found ", B="To String (Integer)"(InInt=@n12), C=" connections.") -> (ReturnValue))
   > Return Node!(execute, SourceTargetMap:Map of Strings to Strings=Get l_SourceTargetMap())
   Set l_sourceData(execute=@n11.LoopBody, l_sourceData:Asset Data Structure=@n11) -> (Output_Get:Asset Data Structure)
   n13 = Break Asset Data(AssetData=@n11)
   > Set l_source(l_source:String="To String (Name)"(InName=@n13)) -> (Output_Get:String)
   > n14 = Get Dependencies!(
    execute
    self:Asset Registry Interface=GetAssetRegistry()
    PackageName:Name=@n13
    DependencyOptions:"Asset Registry Dependency Options Structure (by ref)"=Make Asset Registry Dependency Options!(bIncludeGamePackageReferences, bIncludeEditorOnlyPackageReferences) -> (AssetRegistryDependencyOptions)
   ) -> (then, OutDependencies:Array of Names, ReturnValue:bool) {is=title K2Node_CallFunction}
   > Branch(Condition=Is Valid Index(TargetArray:Array of Names=@n14))
   > n15 = GetAssets(
    self=GetAssetRegistry()
    Filter=MakeARFilter(PackageNames=@n14, bRecursivePaths=true, bRecursiveClasses=false, bIncludeOnlyOnDiskAssets=false)
    bSkipARFilteredAssets=true
   )
   > n16 = For Each Loop(Array=@n15)
   > Set l_targetData(l_targetData:Asset Data Structure=@n16) -> (Output_Get:Asset Data Structure)
   n17 = Break Asset Data(AssetData=Get l_targetData())
   n18 = Get TargetPluginName()
   > Branch(Condition=StartsWith(SourceString="To String (Name)"(InName=@n17), InPrefix=Append!(A=/, B=@n18) -> (ReturnValue)))
   n19 = Get l_targetData()
   > Add(TargetMap=Get l_SourceTargetMap(), Key=Get l_source(), Value="To String (Name)"(InName=Break Asset Data(AssetData=@n19)))
   Reroute Node()
   > Branch(Condition=Contains(
     SearchIn=BreakTopLevelAssetPath!(TopLevelAssetPath:"Top Level Asset Path Structure (by ref)"=@n2.AssetClassPath) -> (PathString:String)
     Substring=MaterialFunction
    ))
   > n20 = Branch(Condition=Contains Item!(TargetSet:Set of Strings=Get l_searchedMatFunctionRefs(), ItemToFind:"String (by ref)"=@n3) -> (ReturnValue:bool) {is=title K2Node_CallFunction})
  ]}
 GetMaterialFunctionDependenciesInTarget {nodes [
   GetMaterialFunctionDependenciesInTarget!() -> (then, InAssetData:"Asset Data Structure (by ref)", TargetPlugin:String, SearchedRefs:Set of Strings) {is=title K2Node_FunctionEntry}
   > n1 = Sequence!(execute) -> (then_0, then_1, then_2:Exec)
   n2 = "To String (Name)"(InName=Break Asset Data!(AssetData=Get InAssetData() -> (InAssetData:"Asset Data Structure (by ref)")) -> (PackageName))
   > n3 = Branch(Condition=Contains Item(TargetSet=Get SearchedRefs() -> (SearchedRefs:Set of Strings), ItemToFind=@n2))
   > Return Node!(execute, FoundRefsToTarget:Set of Strings)
   Add!(execute=@n3.else, TargetSet:Set of Strings=Get SearchedRefs(), NewItem:"String (by ref)"=@n2) -> (then)
   > n4 = Branch(Condition=StartsWith(SourceString=@n2, InPrefix=Get TargetPlugin() -> (TargetPlugin:String)))
   n5 = Get l_foundRefs() -> (l_foundRefs:Set of Strings)
   > n6 = Add(TargetSet=@n5, NewItem=@n2)
   Return Node(execute=[@n6.then, @n7.then], FoundRefsToTarget=@n5)
   n7 = Branch(execute=@n4.else, Condition=StartsWith(SourceString=@n2, InPrefix=/Engine/))
   n8 = GetAssetRegistry()
   n9 = Get Dependencies(
    execute=@n1.then_1
    self=@n8
    PackageName=Break Asset Data(AssetData=Get InAssetData())
    DependencyOptions=Make Asset Registry Dependency Options()
   )
   > Branch(Condition=Is Valid Index(TargetArray=@n9))
   > n10 = GetAssets(
    self=@n8
    Filter=MakeARFilter(PackageNames=@n9, bRecursivePaths=false, bRecursiveClasses=false, bIncludeOnlyOnDiskAssets=false)
    bSkipARFilteredAssets=true
   )
   > n11 = For Each Loop(Array=@n10)
   > n12 = GetMaterialFunctionDependenciesInTarget(InAssetData=@n11, TargetPlugin=Get TargetPlugin(), SearchedRefs=Get SearchedRefs()) {is=title K2Node_CallFunction}
   > n13 = To Array(A=@n12)
   > Add Items!(execute, TargetSet:Set of Strings=Get l_foundRefs(), NewItems:Array of Strings=@n13) -> (then)
   Return Node(execute=@n1.then_2, FoundRefsToTarget=Get l_foundRefs())
  ]}
 PopulateDisplayList {nodes [
   n1 = PopulateDisplayList!() -> (then, SourceTargetMap:Map of Strings to Strings) {is=title K2Node_FunctionEntry}
   > n2 = Sequence!(execute) -> (then_0, then_1)
   > ClearListItems(self=Get CommonListView_19() -> (CommonListView_19:Common List View Object Reference))
   n3 = Keys!(execute=@n2.then_1, TargetMap:Map of Strings to Strings=@n1) -> (then, Keys:Array of Strings)
   > n4 = Values!(execute, TargetMap:Map of Strings to Strings=@n1) -> (then, Values:Array of Strings)
   > n5 = For Each Loop(Array:Array of Strings=@n4) -> (Array Element:String)
   > n6 = Construct Object from Class!(
    execute
    Class=&/O_PCF_ConnectionDisplayData.O_PCF_ConnectionDisplayData_C
    self
    source_package_name:String="Get (a copy)"(Array=@n3, Dimension 1=@n5.Array Index)
    target_package_name:String=@n5
   ) -> (then, ReturnValue:O PCF Connection Display Data Object Reference)
   > AddItem!(execute, self:List View Object Reference=Get CommonListView_19(), Item:Object Reference=@n6) -> (then)
  ]}
]
variables [
 SourcePluginName {type=string, default=PiewPistolGameplay}
 TargetPluginName {type=string, default=BRCosmetics}
]
widgets {
 info {rootWidgetClass=&StackBox, widgetCount=6}
 widgets [
  StackBox_0 {widgetClassPath=&StackBox, widget=StackBox_0, properties {orientation=Orient_Vertical}}
  Border_12 {
   widgetClassPath=&Border
   parent=StackBox_0
   widget=Border_12
   properties {brushColor=<rgba 0.0284 0.0284 0.0284 1>}
   slot {class=StackBoxSlot}
  }
  txtTitle {
   widgetClassPath=&TextBlock
   bIsVariable=true
   parent=Border_12
   widget=txtTitle
   properties {
    text=Pick a Source and Target
    font {
     bForceMonospaced=false
     bMaterialIsStencil=false
     fontMaterial=null
     fontObject=&Roboto
     letterSpacing=0
     monospacedWidth=1
     size=13.5
     skewAmount=0
     typefaceFontName=Regular
     outlineSettings {
      bApplyOutlineToDropShadows=false
      bMiteredCorners=false
      bSeparateFillAlpha=false
      outlineMaterial=null
      outlineSize=0
      outlineColor=<rgba 0 0 0 1>
     }
    }
   }
   slot {class=BorderSlot}
  }
  detailsSourceAndTarget {
   widgetClassPath=&DetailsView
   parent=StackBox_0
   widget=detailsSourceAndTarget
   properties {bShowScrollBar=false, columnWidth=0.75, propertiesToShow [SourcePluginName, TargetPluginName]}
   slot {class=StackBoxSlot}
  }
  btnFindConnections {
   widgetClassPath=&TA_Generic01_Button
   parent=StackBox_0
   widget=btnFindConnections
   properties {label=Find Connections}
   slot {class=StackBoxSlot}
  }
  CommonListView_19 {
   widgetClassPath=&CommonListView
   bIsVariable=true
   parent=StackBox_0
   widget=CommonListView_19
   properties {
    entryWidgetClass=&/EUW_PCF_ConnectionDisplayItem.EUW_PCF_ConnectionDisplayItem_C
    verticalEntrySpacing=10
    scrollBarStyle {
     thickness=8
     draggedThumbImage {
      bIsDynamicallyLoaded=false
      drawAs=RoundedBox
      imageType=NoImage
      mirroring=NoMirror
      resourceName=None
      resourceObject=null
      tiling=NoTile
      imageSize=<0 0>
      margin=<ltrb 0 0 0 0>
      outlineSettings {
       bUseBrushTransparency=false
       roundingType=FixedRadius
       width=0
       color {colorUseRule=UseColor_Specified, specifiedColor=<rgba 0 0 0 0>}
       cornerRadii=<4 4 4 4>
      }
      tintColor {colorUseRule=UseColor_Specified, specifiedColor=<rgba 0.2159 0.2159 0.2159 1>}
      uVRegion {bIsValid=false, max=<0 0>, min=<0 0>}
     }
     horizontalBackgroundImage {
      bIsDynamicallyLoaded=false
      drawAs=NoDrawType
      imageType=NoImage
      mirroring=NoMirror
      resourceName=None
      resourceObject=null
      tiling=NoTile
      imageSize=<0 0>
      margin=<ltrb 0 0 0 0>
      outlineSettings {
       bUseBrushTransparency=false
       roundingType=HalfHeightRadius
       width=0
       color {colorUseRule=UseColor_Specified, specifiedColor=<rgba 0 0 0 0>}
       cornerRadii=<0 0 0 0>
      }
      tintColor {colorUseRule=UseColor_Specified, specifiedColor=<rgba 1 1 1 1>}
      uVRegion {bIsValid=false, max=<0 0>, min=<0 0>}
     }
     horizontalBottomSlotImage {
      bIsDynamicallyLoaded=false
      drawAs=NoDrawType
      imageType=NoImage
      mirroring=NoMirror
      resourceName=None
      resourceObject=null
      tiling=NoTile
      imageSize=<0 0>
      margin=<ltrb 0 0 0 0>
      outlineSettings {
       bUseBrushTransparency=false
       roundingType=HalfHeightRadius
       width=0
       color {colorUseRule=UseColor_Specified, specifiedColor=<rgba 0 0 0 0>}
       cornerRadii=<0 0 0 0>
      }
      tintColor {colorUseRule=UseColor_Specified, specifiedColor=<rgba 1 1 1 1>}
      uVRegion {bIsValid=false, max=<0 0>, min=<0 0>}
     }
     horizontalTopSlotImage {
      bIsDynamicallyLoaded=false
      drawAs=NoDrawType
      imageType=NoImage
      mirroring=NoMirror
      resourceName=None
      resourceObject=null
      tiling=NoTile
      imageSize=<0 0>
      margin=<ltrb 0 0 0 0>
      outlineSettings {
       bUseBrushTransparency=false
       roundingType=HalfHeightRadius
       width=0
       color {colorUseRule=UseColor_Specified, specifiedColor=<rgba 0 0 0 0>}
       cornerRadii=<0 0 0 0>
      }
      tintColor {colorUseRule=UseColor_Specified, specifiedColor=<rgba 1 1 1 1>}
      uVRegion {bIsValid=false, max=<0 0>, min=<0 0>}
     }
     hoveredThumbImage {
      bIsDynamicallyLoaded=false
      drawAs=RoundedBox
      imageType=NoImage
      mirroring=NoMirror
      resourceName=None
      resourceObject=null
      tiling=NoTile
      imageSize=<0 0>
      margin=<ltrb 0 0 0 0>
      outlineSettings {
       bUseBrushTransparency=false
       roundingType=FixedRadius
       width=0
       color {colorUseRule=UseColor_Specified, specifiedColor=<rgba 0 0 0 0>}
       cornerRadii=<4 4 4 4>
      }
      tintColor {colorUseRule=UseColor_Specified, specifiedColor=<rgba 0.2159 0.2159 0.2159 1>}
      uVRegion {bIsValid=false, max=<0 0>, min=<0 0>}
     }
     normalThumbImage {
      bIsDynamicallyLoaded=false
      drawAs=RoundedBox
      imageType=NoImage
      mirroring=NoMirror
      resourceName=None
      resourceObject=null
      tiling=NoTile
      imageSize=<0 0>
      margin=<ltrb 0 0 0 0>
      outlineSettings {
       bUseBrushTransparency=false
       roundingType=FixedRadius
       width=0
       color {colorUseRule=UseColor_Specified, specifiedColor=<rgba 0 0 0 0>}
       cornerRadii=<4 4 4 4>
      }
      tintColor {colorUseRule=UseColor_Specified, specifiedColor=<rgba 0.0953 0.0953 0.0953 1>}
      uVRegion {bIsValid=false, max=<0 0>, min=<0 0>}
     }
     verticalBackgroundImage {
      bIsDynamicallyLoaded=false
      drawAs=NoDrawType
      imageType=NoImage
      mirroring=NoMirror
      resourceName=None
      resourceObject=null
      tiling=NoTile
      imageSize=<0 0>
      margin=<ltrb 0 0 0 0>
      outlineSettings {
       bUseBrushTransparency=false
       roundingType=HalfHeightRadius
       width=0
       color {colorUseRule=UseColor_Specified, specifiedColor=<rgba 0 0 0 0>}
       cornerRadii=<0 0 0 0>
      }
      tintColor {colorUseRule=UseColor_Specified, specifiedColor=<rgba 1 1 1 1>}
      uVRegion {bIsValid=false, max=<0 0>, min=<0 0>}
     }
     verticalBottomSlotImage {
      bIsDynamicallyLoaded=false
      drawAs=NoDrawType
      imageType=NoImage
      mirroring=NoMirror
      resourceName=None
      resourceObject=null
      tiling=NoTile
      imageSize=<0 0>
      margin=<ltrb 0 0 0 0>
      outlineSettings {
       bUseBrushTransparency=false
       roundingType=HalfHeightRadius
       width=0
       color {colorUseRule=UseColor_Specified, specifiedColor=<rgba 0 0 0 0>}
       cornerRadii=<0 0 0 0>
      }
      tintColor {colorUseRule=UseColor_Specified, specifiedColor=<rgba 1 1 1 1>}
      uVRegion {bIsValid=false, max=<0 0>, min=<0 0>}
     }
     verticalTopSlotImage {
      bIsDynamicallyLoaded=false
      drawAs=NoDrawType
      imageType=NoImage
      mirroring=NoMirror
      resourceName=None
      resourceObject=null
      tiling=NoTile
      imageSize=<0 0>
      margin=<ltrb 0 0 0 0>
      outlineSettings {
       bUseBrushTransparency=false
       roundingType=HalfHeightRadius
       width=0
       color {colorUseRule=UseColor_Specified, specifiedColor=<rgba 0 0 0 0>}
       cornerRadii=<0 0 0 0>
      }
      tintColor {colorUseRule=UseColor_Specified, specifiedColor=<rgba 1 1 1 1>}
      uVRegion {bIsValid=false, max=<0 0>, min=<0 0>}
     }
    }
   }
   slot {class=StackBoxSlot, properties {size {sizeRule=Fill, value=1}}}
  }
 ]
}
